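# Translucent deferred decal material for a water stream, in a Tmp folder of the WaterStreamGenerator tool and flagged for Niagara ribbons.
# Two tiling wave normals are sampled on scaled UVs, flattened and angle-corrected blended.
# Perlin noise, distorted by a second noise, warps UV edge-fade masks that shape the stream.
# UseCircularMeltPool blends that mask toward noise-eroded radial sphere masks and constant colors toward particle color.
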Material /Game/Tools/WaterStreamGenerator/Tmp/M_DecalStream
graph {
 properties {
  bUsedWithNiagaraRibbons=true
  blendMode=BLEND_Translucent
  materialDomain=MD_DeferredDecal
  pixelDepthOffsetMode=PDOM_Legacy
  usedShadingModels=Default Lit
 }
 nodes [
  n1 = TextureObjectParameter() {
   group=WaterNormal
   parameterName=WaterNormal_Texture
   samplerType=SAMPLERTYPE_Normal
   sortPriority=0
   texture=/FortWater/Textures/T_Water_TilingNormal_Waves_01
  }
  n2 = VectorParameter() {group=WaterNormal, parameterName=WaterNormal_UVControl, channelNames {a=VPan, b=UPan, g=VScale, r=UScale}, defaultValue=<rgba 3 3 0 0.25>}
  n3 = ScalarParameter() {defaultValue=1, parameterName=ScaleU}
  n4 = ScalarParameter() {defaultValue=2, parameterName=awd}
  n5 = Multiply(
   A=TextureCoordinate()
   B=MaterialFunctionCall(X=Constant() {r=1}, Y=Multiply(A=@n3, B=@n4)) {materialFunction=&MakeFloat2}.Result
  )
  n6 = TextureObjectParameter() {
   group=WaterNormalB
   parameterName=WaterNormalB_Texture
   samplerType=SAMPLERTYPE_Normal
   sortPriority=0
   texture=&T_Water_TilingNormal_Waves_01
  }
  n7 = VectorParameter() {group=WaterNormalB, parameterName=WaterNormalB_UVControl, channelNames {a=VPan, b=UPan, g=VScale, r=UScale}, defaultValue=<rgba 1 1 0 0.1>}
  n8 = MaterialFunctionCall(
   BaseNormal=MaterialFunctionCall(
    Normal=MaterialFunctionCall(Texture=@n1, UVControl_RGB=@n2, UVControl_A=@n2, "[Override]UVInput"=@n5) {materialFunction=&FXCommon_TextureSample}.RGB
    Flatness=ScalarParameter() {defaultValue=0.6, parameterName=NormalFlat}
   ) {materialFunction=&FlattenNormal}.Result
   AdditionalNormal=MaterialFunctionCall(
    Normal=MaterialFunctionCall(Texture=@n6, UVControl_RGB=@n7, UVControl_A=@n7, "[Override]UVInput"=@n5) {materialFunction=&FXCommon_TextureSample}.RGB
    Flatness=ScalarParameter() {defaultValue=0.8, parameterName=NormalBFlat}
   ) {materialFunction=&FlattenNormal}.Result
  ) {materialFunction=&BlendAngleCorrectedNormals}
  Add(
   A=RotateAboutAxis(
    NormalizedRotationAxis=Constant3Vector() {constant=<rgba 1 0 0 0>}
    RotationAngle=ScalarParameter() {defaultValue=-0.05, parameterName=15}
    PivotPoint=Constant3Vector()
    Position=@n8.Result
   )
   B=@n8.Result
  )
  Constant() {r=1}
  Constant()
  Constant()
  DynamicParameter() {paramNames [TexA_U,TexA_V,TexB_U,TexB_V]}
  Fresnel() {baseReflectFraction=0.5, exponent=1}
  n9 = TextureObjectParameter() {
   group=NoiseA
   parameterName=NoiseA_Texture
   samplerType=SAMPLERTYPE_LinearGrayscale
   sortPriority=0
   texture=&T_SimplePerlin_Greyscale
  }
  n10 = TextureObjectParameter() {
   group=NoiseInNoise
   parameterName=NoiseInNoise_Texture
   samplerType=SAMPLERTYPE_LinearGrayscale
   sortPriority=0
   texture=&T_SimplePerlin_Greyscale
  }
  n11 = VectorParameter() {group=NoiseInNoise, parameterName=NoiseInNoise_UVControl, channelNames {a=VPan, b=UPan, g=VScale, r=UScale}, defaultValue=<rgba 1 1 0 0.1>}
  n12 = MaterialFunctionCall(Texture=@n10, UVControl_RGB=@n11, UVControl_A=@n11, "[Override]UVInput"=@n5) {materialFunction=&FXCommon_TextureSample}
  n13 = VectorParameter() {
   group=NoiseInNoise
   parameterName=NoiseInNoise_NoiseControl
   sortPriority=60
   channelNames {a=VStrength, b=UStrength, g=VOffset, r=UOffset}
   defaultValue=<rgba -0.5 -0.5 0.3 0>
  }
  n14 = VectorParameter() {group=NoiseA, parameterName=NoiseA_UVControl, channelNames {a=VPan, b=UPan, g=VScale, r=UScale}, defaultValue=<rgba 0.74 0.7 0 0.25>}
  n15 = MaterialFunctionCall(
   Texture=@n9
   AdditionalUVOffset=MaterialFunctionCall!(
    NoiseValue=MaterialFunctionCall(R=@n12.R, G=@n12.R) {materialFunction=&AppendMany}.RG
    Control_RGB=@n13
    Control_A=@n13
   ) {materialFunction=/Game/Effects/Niagara/MasterMaterials/Functions/Common/FXCommon_NoiseControl}.NoiseValue
   UVControl_RGB=@n14
   UVControl_A=@n14
   "[Override]UVInput"=@n5
  ) {materialFunction=&FXCommon_TextureSample}
  n16 = MaterialFunctionCall(R=@n15.R, G=@n15.R) {materialFunction=&AppendMany}
  n17 = VectorParameter() {
   group=NoiseA
   parameterName=NoiseA_NoiseControl
   sortPriority=60
   channelNames {a=VStrength, b=UStrength, g=VOffset, r=UOffset}
   defaultValue=<rgba -0.5 -0.5 0.2 0>
  }
  n18 = MaterialFunctionCall(Float2=Add(
    A=TextureCoordinate()
    B=MaterialFunctionCall(NoiseValue=@n16.RG, Control_RGB=@n17, Control_A=@n17) {materialFunction=&FXCommon_NoiseControl}.NoiseValue
   )) {materialFunction=&BreakOutFloat2Components}
  n19 = VectorParameter() {group=EdgeFade, parameterName=EdgeFade_EdgeMask_ControlU, channelNames {a="", b=Power, g=Multiply, r=Offset}, defaultValue=<rgba 0 2.2 3 1>}
  n20 = MaterialFunctionCall!(UorV=@n18.R, EdgeControl=@n19) {materialFunction=/Game/Effects/Niagara/MasterMaterials/Functions/Common/FXCommon_EdgeMask_Simple}
  n21 = VectorParameter() {group=EdgeFade, parameterName=EdgeFade_EdgeMask_ControlV, channelNames {a="", b=Power, g=Multiply, r=Offset}, defaultValue=<rgba 0 25 1 1>}
  n22 = MaterialFunctionCall(UorV=@n18.G, EdgeControl=@n21) {materialFunction=&FXCommon_EdgeMask_Simple}
  n23 = Constant() {r=0.5}
  n24 = ScalarParameter() {group=CircularMelt, parameterName=TexA_U}
  n25 = ScalarParameter() {group=CircularMelt, parameterName=TexA_V}
  n26 = TextureSample(UVs=Add(
    A=TextureCoordinate()
    B=MaterialFunctionCall(X=@n24, Y=@n25) {materialFunction=&MakeFloat2}.Result
   )) {texture=&AtmosphericCloudNoise01}
  n27 = TextureCoordinate()
  n28 = MaterialFunctionCall(X=ScalarParameter() {group=CircularMelt, parameterName=TexB_U}, Y=ScalarParameter() {group=CircularMelt, parameterName=TexB_V}) {materialFunction=&MakeFloat2}
  n29 = Multiply(
   A=Power(Base=Multiply(A=@n26.R) {constB=3})
   B=TextureSample(UVs=Add(A=@n27, B=@n28.Result)) {samplerType=SAMPLERTYPE_Grayscale, texture=&LowResBlurredNoise}.R
  )
  n30 = Multiply(
   A=SphereMask(A=TextureCoordinate(), B=@n23, Radius=@n23) {hardnessPercent=0}
   B=SphereMask(A=TextureCoordinate(), B=Constant() {r=0.5}, Radius=LinearInterpolate(Alpha=@n29) {constA=0.3}, Hardness=Constant()) {hardnessPercent=0}
  )
  n31 = TextureCoordinate()
  n32 = MaterialFunctionCall(X=ScalarParameter() {group=CircularMelt, parameterName=Erosion_U}, Y=ScalarParameter() {group=CircularMelt, parameterName=Erosion_V}) {materialFunction=&MakeFloat2}
  n33 = Add(
   A=TextureSample(UVs=Add(A=@n31, B=@n32.Result)) {texture=&/T_CausticsNoise}.R
   B=ScalarParameter() {defaultValue=1, parameterName=Erosion}
  )
  n34 = Saturate(None=@n33)
  LinearInterpolate(
   A=Multiply(A=ScalarParameter() {defaultValue=0.7, parameterName=ate}, B=Multiply(A=@n20.Mask, B=@n22.Mask)) {constA=0.5}
   B=Multiply(A=Multiply(A=@n30) {constB=2}, B=@n34)
   Alpha=ScalarParameter() {group=CircularMelt, parameterName=UseCircularMeltPool, sortPriority=-1}
  )
  LinearInterpolate(A=@n8.Result, B=@n8.Result, Alpha=ScalarParameter() {group=CircularMelt, parameterName=UseCircularMeltPool, sortPriority=-1})
  LinearInterpolate(
   A=Constant3Vector() {constant=<rgba 0.5 0.5 0.5 0>}
   B=ParticleColor().RGB
   Alpha=ScalarParameter() {group=CircularMelt, parameterName=UseCircularMeltPool, sortPriority=-1}
  )
  LinearInterpolate(A=Constant3Vector(), B=ParticleColor().RGB, Alpha=ScalarParameter() {group=CircularMelt, parameterName=UseCircularMeltPool, sortPriority=-1})
  MaterialFunctionCall() {materialFunction=&ObjectScale}
  Multiply(A=@n29, B=@n34)
  Subtract(A=@n33, B=DynamicParameter() {parameterIndex=1, paramNames [ErosionTex_U, ErosionTex_V, ErosionAdd, ErosionSubtract]}.ErosionSubtract)
 ]
}
