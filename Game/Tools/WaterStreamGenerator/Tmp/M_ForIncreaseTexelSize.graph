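# Translucent unlit utility material in the WaterStreamGenerator Tmp folder, apparently related to texel size, judging by its name.
# It multiplies the red channels of a drop-smoke texture, sampled as a normal map, and a single-frame texture, then multiplies by the big-flow water stream RGB.
# A separate Opacity parameter defaults to almost zero.
# The graph shows no links from either branch to material outputs.
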Material /Game/Tools/WaterStreamGenerator/Tmp/M_ForIncreaseTexelSize
graph {
 properties {blendMode=BLEND_Translucent, pixelDepthOffsetMode=PDOM_Legacy, shadingModel=MSM_Unlit, usedShadingModels=Unlit}
 nodes [
  Multiply(
   A=Multiply(
    A=TextureSample() {samplerType=SAMPLERTYPE_Normal, texture=/Game/Tools/NewWaterfallGenerator/FX/Texture/T_MergedDropSmoke_8x10}.R
    B=TextureSample() {texture=/Game/Tools/WaterStreamGenerator/Seq/T_SingleFrame2}.R
   )
   B=TextureSample() {texture=/Game/Tools/WaterStreamGenerator/Seq/T_WaterStream_BigFlow}.RGB
  )
  ScalarParameter() {defaultValue=0.0001, parameterName=Opacity}
 ]
}
