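# Translucent lit material for water-stream drop smoke, usable on Niagara meshes and ribbons, with spline thickening driven by parameters.
# An animated flipbook scrolls inside each cell, blending frames with motion-vector warping and masking wrap seams with an offset sample.
# Lower quality and mobile/forward paths skip frame blending; Low uses a single-frame scrolling texture.
# Opacity combines the smoke with scale-aware end fades, depth fade and near-camera fade.
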Material /Game/Tools/WaterStreamGenerator/Tmp/M_MergedDropSmoke
graph {
 properties {
  bUsedWithNiagaraMeshParticles=true
  bUsedWithNiagaraRibbons=true
  blendMode=BLEND_Translucent
  pixelDepthOffsetMode=PDOM_Legacy
  usedShadingModels=Default Lit
 }
 nodes [
  Constant() {r=1}
  Constant() {r=1}
  n1 = TextureCoordinate()
  DDX(Value=@n1)
  DDY(Value=@n1)
  DynamicParameter() {paramNames [OffsetTime,Param2,Param3,Param4]}
  n2 = VectorParameter() {parameterName=Spline, defaultValue=<rgba 200 50 0 1>}
  n3 = ScalarParameter() {parameterName=ScaleY}
  MaterialFunctionCall(WidthBase=Multiply(A=@n2.R, B=@n3), WidthTip=Multiply(A=@n2.G, B=@n3), Expand U or V UV Channel=StaticBool()) {materialFunction=&SplineThicken}
  n4 = VectorParameter() {parameterName=Color, defaultValue=<rgba 0.75 0.9167 1 1.2>}
  Multiply(A=@n4.RGB, B=ScalarParameter() {defaultValue=0.1, parameterName=E})
  n5 = MaterialFunctionCall(Float2=TextureCoordinate()) {materialFunction=&BreakOutFloat2Components}
  n6 = MaterialFunctionCall() {materialFunction=&ObjectScale}
  n7 = Multiply(A=@n5.R, B=@n6.Scale Z)
  n8 = ScalarParameter() {defaultValue=5, parameterName=TopFadeDistance}
  n9 = Multiply(A=OneMinus(None=@n5.R), B=@n6.Scale Z)
  n10 = ScalarParameter() {defaultValue=1, parameterName=BottomFadeDistance}
  n11 = Multiply(A=Saturate(None=Multiply(A=@n7, B=@n8)), B=Saturate(None=Multiply(A=@n9, B=@n10)))
  n12 = Add(A=Time() {bOverride_Period=true, period=6666}, B=ScalarParameter() {parameterName=RandomOffset})
  n13 = VectorParameter() {parameterName=SubImage, defaultValue=<rgba 0 0 0 1>}
  n14 = Fmod(A=Multiply(A=ScalarParameter() {parameterName=Speed}, B=@n12), B=Subtract(A=Multiply(A=@n13.R, B=@n13.R)))
  n15 = Frac(None=@n14) {desc=Frame Lerp}
  n16 = Subtract(A=@n14, B=@n15) {desc=IndexA}
  n17 = AppendVector(A=@n13.R, B=@n13.R)
  n18 = ComponentMask(None=@n17) {desc="## Horz Images", r=true}
  n19 = Divide(A=Constant2Vector() {g=1, r=1}, B=@n17) {desc=Scale}
  n20 = ComponentMask(None=@n19) {desc="1 / ## Horz Images", r=true}
  n21 = AppendVector(A=Fmod(A=@n16, B=@n18) {desc=IndexA_Horz}, B=Floor(None=Multiply(A=@n16, B=@n20)) {desc=IndexA_Vert}) {desc=IndexA_Combined}
  n22 = MaterialFunctionCall(Float2=TextureCoordinate()) {materialFunction=&BreakOutFloat2Components}
  n23 = VectorParameter() {parameterName=SCaleUV, defaultValue=<rgba 0 0 0 1>}
  n24 = ScalarParameter() {defaultValue=1, parameterName=ScaleZ}
  n25 = Multiply(
   A=MaterialFunctionCall(X=@n22.G, Y=@n22.G) {materialFunction=&MakeFloat2}.Result
   B=AppendVector(A=@n23.R, B=Multiply(A=@n23.G, B=@n24))
  )
  n26 = Multiply(A=@n21, B=@n19)
  n27 = Subtract(A=Multiply(A=Add(A=@n21, B=@n25), B=@n19) {desc=UV_A}, B=@n26)
  n28 = MaterialFunctionCall(Float2=@n27) {materialFunction=&BreakOutFloat2Components}
  n29 = VectorParameter() {parameterName=TextureSize, defaultValue=<rgba 512 1024 0 1>}
  n30 = MaterialFunctionCall(Float2=Divide(A=Constant2Vector() {g=1, r=1}, B=AppendVector(A=@n29.R, B=@n29.R)) {desc=Scale}) {materialFunction=&BreakOutFloat2Components}
  n31 = ScalarParameter() {parameterName=OffsetPixel}
  n32 = Multiply(A=ScalarParameter() {parameterName=Offset}, B=@n12)
  n33 = Add(A=@n28.G, B=Add(A=@n32, B=Multiply(A=@n30.G, B=@n31)))
  n34 = Multiply(A=@n30.G) {constB=0.5}
  n35 = Subtract(
   A=MaterialFunctionCall(Float2=@n19) {materialFunction=&BreakOutFloat2Components}.G
   B=@n34
  )
  n36 = Fmod(A=Abs(None=@n33), B=@n35)
  n37 = Clamp(None=If(A=@n33, "A > B"=@n36, "A == B"=@n36, "A < B"=Subtract(A=@n35, B=@n36))) {clampMode=CMODE_ClampMin}
  n38 = TextureObjectParameter() {parameterName=Tex, samplerType=SAMPLERTYPE_Normal, texture=/Game/Tools/NewWaterfallGenerator/FX/Texture/T_MergedDropSmoke_8x10}
  n39 = TextureSample(UVs=Add(A=AppendVector(A=@n28.R, B=@n37), B=@n26), Tex=@n38) {samplerSource=SSM_Wrap_WorldGroupSettings, texture=/Engine/EngineMaterials/DefaultDiffuse}
  n40 = TextureObjectParameter() {parameterName=Tex_MV, samplerType=SAMPLERTYPE_Normal, texture=/Game/Tools/NewWaterfallGenerator/FX/Texture/T_MergedDropSmokeWithMV_MV}
  n41 = MaterialFunctionCall(Float2=@n27) {materialFunction=&BreakOutFloat2Components}
  n42 = Add(A=@n41.G, B=@n32)
  n43 = Fmod(A=Abs(None=@n42), B=@n35)
  n44 = Clamp(None=If(A=@n42, "A > B"=@n43, "A == B"=@n43, "A < B"=Subtract(A=@n35, B=@n43)), Min=@n34) {clampMode=CMODE_ClampMin}
  n45 = Add(A=AppendVector(A=@n41.R, B=@n44), B=@n26)
  n46 = TextureSample(UVs=@n45, Tex=@n40) {samplerSource=SSM_Wrap_WorldGroupSettings, texture=&DefaultDiffuse}
  n47 = Divide(A=Multiply(A=ComponentMask(None=@n46.RGB) {g=true, r=true}, B=ScalarParameter() {parameterName=MVIns}), B=@n17)
  n48 = TextureSample(UVs=Add(A=@n45, B=Multiply(A=@n47, B=@n15)), Tex=@n38) {samplerSource=SSM_Wrap_WorldGroupSettings, texture=&DefaultDiffuse}
  n49 = Add(B=@n16) {constA=1, desc=IndexB}
  n50 = AppendVector(A=Fmod(A=@n49, B=@n18) {desc=IndexB_Horz}, B=Floor(None=Multiply(A=@n49, B=@n20)) {desc=IndexB_Vert}) {desc=IndexB_Combined}
  n51 = Add(A=@n25, B=@n50)
  n52 = Multiply(A=@n50, B=@n19)
  n53 = MaterialFunctionCall(Float2=Subtract(A=Multiply(A=@n51, B=@n19) {desc=UV_B}, B=@n52)) {materialFunction=&BreakOutFloat2Components}
  n54 = Add(A=@n53.G, B=@n32)
  n55 = Fmod(A=Abs(None=@n54), B=@n35)
  n56 = Subtract(A=@n35, B=@n55)
  n57 = AppendVector(A=@n53.R, B=Clamp(None=If(A=@n54, "A > B"=@n55, "A == B"=@n55, "A < B"=@n56), Min=@n34) {clampMode=CMODE_ClampMin})
  n58 = LinearInterpolate(
   A=@n48.R
   B=TextureSample(UVs=Add(A=@n57, B=@n52), Tex=@n38) {samplerSource=SSM_Wrap_WorldGroupSettings, texture=&DefaultDiffuse}.R
   Alpha=@n15
  )
  n59 = ShadingPathSwitch(Default=@n58, Deferred=@n58, Forward=@n48.R, Mobile=@n48.R)
  n60 = QualitySwitch(Default=@n59, Low=@n48.R, High=@n59, Medium=@n48.R, Epic=@n59)
  n61 = Divide(A=@n44, B=@n35)
  n62 = ScalarParameter() {defaultValue=100, parameterName=Hardness}
  n63 = Multiply(A=OneMinus(None=@n61), B=@n62) {constB=15}
  n64 = ScalarParameter() {defaultValue=2, parameterName=P}
  n65 = Multiply(A=@n61, B=@n62) {constB=15}
  n66 = Min(A=Power(Base=Saturate(None=@n63), Exp=@n64), B=Power(Base=Saturate(None=@n65), Exp=@n64))
  n67 = Multiply(A=Add(A=LinearInterpolate(A=@n39.R, B=@n60, Alpha=@n66) {constA=-1})) {constB=0.5}
  n68 = ShadingPathSwitch(Default=@n67, Deferred=@n67, Forward=@n67, Mobile=@n67)
  n69 = MaterialFunctionCall(Float2=@n25) {materialFunction=&BreakOutFloat2Components}
  n70 = Multiply(A=ScalarParameter() {defaultValue=15, parameterName=LowPlatFormOffsetScale}, B=@n32)
  n71 = TextureSample(UVs=AppendVector(A=@n69.R, B=Add(A=@n69.G, B=@n70))) {samplerSource=SSM_Wrap_WorldGroupSettings, texture=/Game/Tools/NewWaterfallGenerator/FX/Texture/T_SingleFrameFlowSmoke}
  n72 = DepthFade(
   Opacity=Multiply(A=QualitySwitch(Default=@n68, Low=@n71.R, High=@n68, Medium=@n68, Epic=@n68), B=@n4.A)
   FadeDistance=ScalarParameter() {parameterName=DF}
  )
  n73 = Distance(A=CameraPositionWS(), B=WorldPosition().XYZ)
  n74 = VectorParameter() {parameterName=CameraFade, defaultValue=<rgba 200 1000 0 1>}
  n75 = Saturate(None=Divide(A=Subtract(A=@n73, B=@n74.R), B=@n74.G))
  Multiply(A=Multiply(A=Multiply(A=@n11, B=@n72), B=ScalarParameter() {defaultValue=1.5, parameterName=OpaScale}), B=@n75)
  OneMinus(None=@n75)
 ]
}
