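# Translucent Default Lit material for Niagara sprite water splashes, stored in the WaterStreamGenerator tool's Tmp folder.
# The red channel of a baked side-splash atlas texture from NewWaterfallGenerator runs through parameterized add, multiply and power steps, with saturation, to reshape its grayscale mask.
# That mask is multiplied by particle alpha and fed through a parameterized DepthFade to soften intersections with scene geometry.
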Material /Game/Tools/WaterStreamGenerator/Tmp/M_Splash
graph {
 properties {bUsedWithNiagaraSprites=true, blendMode=BLEND_Translucent, pixelDepthOffsetMode=PDOM_Legacy, usedShadingModels=Default Lit}
 nodes [
  Constant() {r=1}
  n1 = Add(
   A=TextureSample() {samplerType=SAMPLERTYPE_LinearColor, texture=/Game/Tools/NewWaterfallGenerator/FX/Texture/SideSplash1_BakedAtlas_Texture2D_0}.R
   B=ScalarParameter() {group=MainT, parameterName=MainT_GrayAdjust_Add}
  )
  n2 = ScalarParameter() {defaultValue=3, group=MainT, parameterName=MainT_GrayAdjust_Mul}
  n3 = Power(Base=Saturate(None=Multiply(A=@n1, B=@n2)), Exp=ScalarParameter() {defaultValue=4, group=MainT, parameterName=MainT_GrayAdjust_Power})
  DepthFade(Opacity=Multiply(A=Saturate(None=@n3), B=ParticleColor().A), FadeDistance=ScalarParameter() {defaultValue=50, parameterName=DF})
 ]
}
