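# Translucent, two-sided lit material for Niagara sprites that renders flowing water from animated SubUV flipbook color and normal textures.
# Dynamic parameters add random time offset, scale V tiling by SizeZ, and drive highlight, opacity and refraction.
# Highlights add to particle color, while opacity combines texture alpha, absorption, a V-axis edge mask and depth fade.
# Boosted flipbook normals add surface detail, and masked index-of-refraction distortion fades with camera distance.
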Material /Game/Tools/WaterStreamGenerator/Tmp/M_Stream
graph {
 properties {
  bUsedWithNiagaraSprites=true
  blendMode=BLEND_Translucent
  pixelDepthOffsetMode=PDOM_Legacy
  refractionMethod=RM_IndexOfRefraction
  translucencyLightingMode=TLM_Surface
  twoSided=true
  usedShadingModels=Default Lit
 }
 nodes [
  Constant()
  n1 = VectorParameter() {group=BaseUV, parameterName=BaseUV_UVOffset, channelNames {a=0, b=0, g=VOffset, r=UOffset}}
  n2 = VectorParameter() {group=BaseUV, parameterName=BaseUV_UVControl, channelNames {a=VPan, b=UPan, g=VScale, r=UScale}, defaultValue=<rgba 1 1 0 0>}
  MaterialExpressionAppend3Vector_1 = MaterialXAppend3Vector!(A=@n2.UScale, B=Multiply(A=@n2.VScale, B=DynamicParameter() {parameterIndex=1, paramNames [SizeZ,0,0,0]}.SizeZ), C=@n2.UScale)
  n3 = DynamicParameter() {paramNames [HighLight,Opactiy,Ref,Rand]}
  n4 = MaterialFunctionCall!(
   Texture
   AdditionalUVOffset=AppendVector(A=@n1.UOffset, B=@n1.UOffset)
   UVControl_RGB=@MaterialExpressionAppend3Vector_1
   UVControl_A=@n2.VPan
   "[Override]TimeInput"=Add(A=@n3.Rand, B=Time() {bOverride_Period=true, period=9999})
   "[Override]ScaleUVByCenter"=StaticBool() {value=true}
   "[Override]UVInput"
  ) {materialFunction=/Game/Effects/Niagara/MasterMaterials/Functions/Common/FXCommon_TextureSample_V2}
  n5 = VectorParameter() {parameterName=SubImages, defaultValue=<rgba 7 7 32 1>}
  n6 = AppendVector(A=@n5.R, B=@n5.R)
  n7 = Multiply(A=@n5.B, B=Time() {bOverride_Period=true, period=6666})
  n8 = MaterialFunctionCall(Float3=MaterialFunctionCall!(
    Texture=TextureObjectParameter() {parameterName=Param_1, samplerType=SAMPLERTYPE_Normal, texture=/Game/Tools/WaterStreamGenerator/Seq/T_WaterStream_TopFlow_N}
    UVs=@n4.UV
    SubImages=@n6
    Frame=@n7
   ) {materialFunction=/Engine/Functions/Engine_MaterialFunctions03/Texturing/SubUV_Function}.RGB) {materialFunction=&BreakOutFloat3Components}
  n9 = Multiply(A=Subtract(A=AppendVector(A=@n8.R, B=@n8.R)) {constB=0.5}) {constB=2}
  Multiply(A=@n9) {constB=10}
  n10 = MaterialFunctionCall(
   Texture=TextureObjectParameter() {parameterName=Param, texture=/Game/Tools/WaterStreamGenerator/Seq/T_WaterStream_TopFlow}
   UVs=@n4.UV
   SubImages=@n6
   Frame=@n7
  ) {materialFunction=&SubUV_Function}
  Reroute(None=@n10.RGB)
  Reroute(None=Constant())
  Reroute(None=Constant() {r=1})
  n11 = Multiply(A=ScalarParameter() {defaultValue=3, parameterName=HighlightStr}, B=@n3.HighLight)
  Reroute(None=Add(A=Multiply(A=@n10.RGB, B=@n11), B=ParticleColor().RGB))
  n12 = VectorParameter() {
   group=EdgeMask
   parameterName=EdgeMask_EdgeMask_InvertControl
   channelNames {a="", b=Power, g=Multiply, r=Offset}
   defaultValue=<rgba 0 1 1 1>
  }
  n13 = VectorParameter() {group=EdgeMask, parameterName=EdgeMask_EdgeMask_Control, channelNames {a="", b=Power, g=Multiply, r=Offset}, defaultValue=<rgba 0 15 2 1>}
  n14 = MaterialFunctionCall!(
   UorV=MaterialFunctionCall(Float2=TextureCoordinate()) {materialFunction=&BreakOutFloat2Components}.G
   Invert_EdgeControl=@n12
   NoInvert_EdgeControl=@n13
  ) {materialFunction=/Game/Effects/Niagara/MasterMaterials/Functions/Common/FXCommon_EdgeMask_Ex}
  n15 = Multiply(A=ScalarParameter() {defaultValue=0.6, parameterName=WaterAbsorb}, B=@n3.Opactiy)
  n16 = Multiply(A=@n10.Alpha, B=Saturate(None=Add(A=@n10.RGB, B=@n15)))
  Reroute(None=DepthFade(Opacity=Multiply(A=@n14.Mask, B=@n16), FadeDistance=ScalarParameter() {defaultValue=50, parameterName=DF}))
  Reroute(None=MaterialFunctionCall(Normal=DeriveNormalZ(InXY=@n9), Flatness=Constant() {r=-2}) {materialFunction=&FlattenNormal}.Result)
  n17 = Multiply(A=ScalarParameter() {defaultValue=0.9, parameterName=Refraction}, B=@n3.Ref)
  n18 = WorldPosition()
  n19 = CameraPositionWS()
  n20 = Saturate(None=Divide(A=Distance(A=@n18.XYZ, B=@n19)) {constB=4096})
  Reroute(None=LinearInterpolate(B=LinearInterpolate(A=@n17, Alpha=@n20), Alpha=Multiply(A=@n14.Mask, B=@n10.Alpha)) {constA=1})
 ]
}
