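# Additive, two-sided lit water-stream material, apparently for falling mist, in the WaterStreamGenerator Tmp folder, flagged for Niagara sprites.
# It plays a flipbook with blended frames, scrolling V inside each cell and hiding the wrap seam with an offset sample.
# Low quality and forward/mobile paths swap to a scrolling single frame and flat normal.
# Edge masks, depth fade and camera-distance fade shape opacity; SplineThicken builds the strip from the normal map.
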Material /Game/Tools/WaterStreamGenerator/Tmp/M_Stream_MistFall
graph {
 properties {
  bTangentSpaceNormal=false
  bUsedWithNiagaraSprites=true
  blendMode=BLEND_Additive
  pixelDepthOffsetMode=PDOM_Legacy
  refractionMethod=RM_IndexOfRefraction
  translucencyLightingMode=TLM_Surface
  twoSided=true
  usedShadingModels=Default Lit
 }
 nodes [
  n1 = Add(A=Time() {bOverride_Period=true, period=6666}, B=ScalarParameter() {parameterName=RandomOffset})
  n2 = VectorParameter() {parameterName=SubImage, defaultValue=<rgba 8 8 0 1>}
  n3 = Fmod(A=Multiply(A=ScalarParameter() {defaultValue=28, parameterName=Speed}, B=@n1), B=Subtract(A=Multiply(A=@n2.R, B=@n2.R)))
  n4 = Frac(None=@n3) {desc=Frame Lerp}
  n5 = Subtract(A=@n3, B=@n4) {desc=IndexA}
  n6 = AppendVector(A=@n2.R, B=@n2.R)
  n7 = ComponentMask(None=@n6) {desc="## Horz Images", r=true}
  n8 = Divide(A=Constant2Vector() {g=1, r=1}, B=@n6) {desc=Scale}
  n9 = ComponentMask(None=@n8) {desc="1 / ## Horz Images", r=true}
  n10 = AppendVector(A=Fmod(A=@n5, B=@n7) {desc=IndexA_Horz}, B=Floor(None=Multiply(A=@n5, B=@n9)) {desc=IndexA_Vert}) {desc=IndexA_Combined}
  n11 = MaterialFunctionCall(Float2=TextureCoordinate()) {materialFunction=&BreakOutFloat2Components}
  n12 = VectorParameter() {parameterName=SCaleUV, defaultValue=<rgba 1 0.25 0 1>}
  n13 = ScalarParameter() {defaultValue=1, parameterName=ScaleZ}
  n14 = Multiply(
   A=MaterialFunctionCall(X=@n11.G, Y=@n11.G) {materialFunction=&MakeFloat2}.Result
   B=AppendVector(A=@n12.R, B=Multiply(A=@n12.G, B=@n13))
  )
  n15 = Multiply(A=@n10, B=@n8)
  n16 = Subtract(A=Multiply(A=Add(A=@n10, B=@n14), B=@n8) {desc=UV_A}, B=@n15)
  n17 = MaterialFunctionCall(Float2=@n16) {materialFunction=&BreakOutFloat2Components}
  n18 = VectorParameter() {parameterName=TextureSize, defaultValue=<rgba 512 1024 0 1>}
  n19 = MaterialFunctionCall(Float2=Divide(A=Constant2Vector() {g=1, r=1}, B=AppendVector(A=@n18.R, B=@n18.R)) {desc=Scale}) {materialFunction=&BreakOutFloat2Components}
  n20 = ScalarParameter() {defaultValue=3, parameterName=OffsetPixel}
  n21 = Multiply(A=ScalarParameter() {parameterName=Offset}, B=@n1)
  n22 = Add(A=@n17.G, B=Add(A=@n21, B=Multiply(A=@n19.G, B=@n20)))
  n23 = Multiply(A=@n19.G) {constB=0.5}
  n24 = Subtract(
   A=MaterialFunctionCall(Float2=@n8) {materialFunction=&BreakOutFloat2Components}.G
   B=@n23
  )
  n25 = Fmod(A=Abs(None=@n22), B=@n24)
  n26 = Subtract(A=@n24, B=@n25)
  n27 = AppendVector(A=@n17.R, B=Clamp(None=If(A=@n22, "A > B"=@n25, "A == B"=@n25, "A < B"=@n26)) {clampMode=CMODE_ClampMin})
  n28 = TextureObjectParameter() {parameterName=Tex, texture=/Game/Tools/WaterStreamGenerator/Seq/T_WaterStream_BigFlow}
  n29 = MaterialFunctionCall(Float2=@n16) {materialFunction=&BreakOutFloat2Components}
  n30 = Add(A=@n29.G, B=@n21)
  n31 = Fmod(A=Abs(None=@n30), B=@n24)
  n32 = Clamp(None=If(A=@n30, "A > B"=@n31, "A == B"=@n31, "A < B"=Subtract(A=@n24, B=@n31)), Min=@n23) {clampMode=CMODE_ClampMin}
  n33 = Add(A=AppendVector(A=@n29.R, B=@n32), B=@n15)
  n34 = TextureSample(UVs=@n33, Tex=@n28) {samplerSource=SSM_Wrap_WorldGroupSettings, texture=/Engine/EngineMaterials/DefaultDiffuse}
  n35 = Add(B=@n5) {constA=1, desc=IndexB}
  n36 = AppendVector(A=Fmod(A=@n35, B=@n7) {desc=IndexB_Horz}, B=Floor(None=Multiply(A=@n35, B=@n9)) {desc=IndexB_Vert}) {desc=IndexB_Combined}
  n37 = Add(A=@n14, B=@n36)
  n38 = Multiply(A=@n36, B=@n8)
  n39 = MaterialFunctionCall(Float2=Subtract(A=Multiply(A=@n37, B=@n8) {desc=UV_B}, B=@n38)) {materialFunction=&BreakOutFloat2Components}
  n40 = Add(A=@n39.G, B=@n21)
  n41 = Fmod(A=Abs(None=@n40), B=@n24)
  n42 = Subtract(A=@n24, B=@n41)
  n43 = AppendVector(A=@n39.R, B=Clamp(None=If(A=@n40, "A > B"=@n41, "A == B"=@n41, "A < B"=@n42), Min=@n23) {clampMode=CMODE_ClampMin})
  n44 = LinearInterpolate(
   A=@n34.RGBA
   B=TextureSample(UVs=Add(A=@n43, B=@n38), Tex=@n28) {samplerSource=SSM_Wrap_WorldGroupSettings, texture=&DefaultDiffuse}.RGBA
   Alpha=@n4
  )
  n45 = ShadingPathSwitch(Default=@n44, Deferred=@n44, Forward=@n34.RGBA, Mobile=@n34.RGBA)
  n46 = Divide(A=@n32, B=@n24)
  n47 = ScalarParameter() {defaultValue=100, parameterName=Hardness}
  n48 = Saturate(None=Multiply(A=OneMinus(None=@n46), B=@n47) {constB=15})
  n49 = ScalarParameter() {defaultValue=2, parameterName=P}
  n50 = Saturate(None=Multiply(A=@n46, B=@n47) {constB=15})
  n51 = LinearInterpolate(
   A=TextureSample(UVs=Add(A=@n27, B=@n15), Tex=@n28) {samplerSource=SSM_Wrap_WorldGroupSettings, texture=&DefaultDiffuse}.RGBA
   B=QualitySwitch(Default=@n45, Low=@n34.RGBA, High=@n45, Medium=@n34.RGBA, Epic=@n45)
   Alpha=Min(A=Power(Base=@n48, Exp=@n49), B=Power(Base=@n50, Exp=@n49))
  )
  n52 = MaterialFunctionCall(Float2=@n14) {materialFunction=&BreakOutFloat2Components}
  n53 = Add(A=@n52.G, B=Multiply(A=ScalarParameter() {defaultValue=-0.4, parameterName=LowPlatFormOffsetScale}, B=@n1))
  n54 = Multiply(
   A=TextureSample(UVs=AppendVector(A=@n52.R, B=@n53)) {samplerSource=SSM_Wrap_WorldGroupSettings, texture=/Game/Tools/WaterStreamGenerator/Seq/T_SingleFrame2}.RGBA
   B=Constant4Vector() {constant=<rgba 0.5 0.5 0.5 1>}.RGBA
  )
  n55 = ShadingPathSwitch(Default=@n51, Deferred=@n51, Forward=@n54, Mobile=@n54)
  n56 = QualitySwitch(Default=@n55, Low=@n54, High=@n55, Medium=@n55, Epic=@n55)
  n57 = ComponentMask(None=@n56) {r=true}
  Add(
   A=Multiply(A=@n57, B=ScalarParameter() {defaultValue=5, parameterName=HighlightStr})
   B=VectorParameter() {parameterName=C, defaultValue=<rgba 0.021 0.07 0.0455 1>}.RGB
  )
  Constant() {r=1}
  n58 = TextureCoordinate()
  DDX(Value=@n58)
  DDY(Value=@n58)
  n59 = Distance(A=CameraPositionWS(), B=WorldPosition().XYZ)
  n60 = VectorParameter() {parameterName=CameraFade, defaultValue=<rgba 200 3000 0 1>}
  n61 = Saturate(None=Divide(A=Subtract(A=@n59, B=@n60.R), B=@n60.G))
  n62 = ScalarParameter() {defaultValue=0.7, parameterName=Refraction}
  n63 = WorldPosition()
  n64 = CameraPositionWS()
  n65 = Saturate(None=Divide(A=Distance(A=@n63.XYZ, B=@n64)) {constB=4096})
  n66 = MaterialFunctionCall(Float2=TextureCoordinate()) {materialFunction=&BreakOutFloat2Components}
  n67 = VectorParameter() {
   group=EdgeMask
   parameterName=EdgeMask2_EdgeMask_InvertControl
   channelNames {a="", b=Power, g=Multiply, r=Offset}
   defaultValue=<rgba 0 5 2 1>
  }
  n68 = MaterialFunctionCall!(UorV=@n66.G, EdgeControl=@n67) {materialFunction=/Game/Effects/Niagara/MasterMaterials/Functions/Common/FXCommon_EdgeMask_Simple}
  n69 = VectorParameter() {
   group=EdgeMask
   parameterName=EdgeMask_EdgeMask_InvertControl
   channelNames {a="", b=Power, g=Multiply, r=Offset}
   defaultValue=<rgba 0 1.5 2 1>
  }
  n70 = VectorParameter() {group=EdgeMask, parameterName=EdgeMask_EdgeMask_Control, channelNames {a="", b=Power, g=Multiply, r=Offset}, defaultValue=<rgba 0 15 2 1>}
  n71 = MaterialFunctionCall!(UorV=@n66.R, Invert_EdgeControl=@n69, NoInvert_EdgeControl=@n70) {materialFunction=/Game/Effects/Niagara/MasterMaterials/Functions/Common/FXCommon_EdgeMask_Ex}
  n72 = ComponentMask(None=@n56) {a=true}
  n73 = Power(Base=@n72, Exp=ScalarParameter() {defaultValue=1.5, parameterName=Power})
  n74 = Multiply(A=@n71.Mask, B=@n73)
  LinearInterpolate(B=LinearInterpolate(B=LinearInterpolate(A=@n62, Alpha=@n65), Alpha=Multiply(A=@n68.Mask, B=@n74)) {constA=1}, Alpha=@n61) {constA=1}
  n75 = VectorParameter() {parameterName=Spline, defaultValue=<rgba 200 50 0 1>}
  n76 = ScalarParameter() {parameterName=ScaleY}
  n77 = MaterialFunctionCall(Float3=TextureSample(
    UVs=@n33
    Tex=TextureObjectParameter() {parameterName=TexN, samplerType=SAMPLERTYPE_Normal, texture=/Game/Tools/WaterStreamGenerator/Seq/T_WaterStream_BigFlow_N}
   ) {samplerSource=SSM_Wrap_WorldGroupSettings, texture=&DefaultDiffuse}.RGB) {materialFunction=&BreakOutFloat3Components}
  n78 = AppendVector(A=@n77.R, B=@n77.R)
  n79 = DeriveNormalZ(InXY=Multiply(A=Subtract(A=@n78) {constB=0.5}) {constB=2})
  n80 = Constant3Vector() {constant=<rgba 0 0 1 0>}
  n81 = ShadingPathSwitch(Default=@n79, Deferred=@n79, Forward=@n80, Mobile=@n80)
  MaterialFunctionCall(
   WidthBase=Multiply(A=@n75.R, B=@n76)
   WidthTip=Multiply(A=@n75.G, B=@n76)
   Expand U or V UV Channel=StaticBool()
   AdditionalNormal=MaterialFunctionCall(Normal=QualitySwitch(Default=@n81, Low=@n80, High=@n81, Medium=@n81, Epic=@n81), Flatness=Constant() {r=0.5}) {materialFunction=&FlattenNormal}.Result
   FlattenPixelNormal=StaticBool() {value=true}
  ) {materialFunction=&SplineThicken}
  n82 = Saturate(None=Add(A=@n72, B=ScalarParameter() {defaultValue=0.5, parameterName=WaterAbsorb}))
  n83 = Multiply(A=@n68.Mask, B=Multiply(A=@n71.Mask, B=Multiply(A=@n73, B=@n82)))
  n84 = ScalarParameter() {defaultValue=0.5, parameterName=Opa}
  Multiply(A=@n61, B=DepthFade(Opacity=Multiply(A=@n83, B=@n84), FadeDistance=ScalarParameter() {defaultValue=50, parameterName=DF}))
  OneMinus(None=@n61)
  Reroute(None=Multiply(A=@n57, B=ScalarParameter() {defaultValue=1, parameterName=Emi}))
  Reroute(None=Constant())
  Reroute(None=ScalarParameter() {defaultValue=0.05, parameterName=Sp})
 ]
}
